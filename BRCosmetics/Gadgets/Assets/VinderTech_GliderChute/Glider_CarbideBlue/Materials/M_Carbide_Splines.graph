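# Translucent, unlit, two-sided material for glowing animated spline strands on the Carbide Blue glider chute.
# SplineThicken sets strand width from scrolling noise, which also wobbles strands along an offset vector, strongest mid-UV, with vertex green scaling speed.
# Emissive layers a panning ramp pulse over a dark-to-bright gradient, which DepthFade intensifies near intersecting geometry.
# The pulse plus a tweakable static ramp forms an apparent alpha term.
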Material /BRCosmetics/Gadgets/Assets/VinderTech_GliderChute/Glider_CarbideBlue/Materials/M_Carbide_Splines
graph {
 properties {
  bUsedWithMeshDeformer=true
  bUsedWithSkeletalMesh=true
  blendMode=BLEND_Translucent
  pixelDepthOffsetMode=PDOM_Legacy
  shadingModel=MSM_Unlit
  twoSided=true
  usedShadingModels=Unlit
 }
 nodes [
  n1 = TextureCoordinate() {coordinateIndex=1, vTiling=0.5}
  n2 = Multiply(A=ScalarParameter() {defaultValue=-0.2286, parameterName=CoordX}, B=ComponentMask(None=@n1) {r=true})
  n3 = VertexColor()
  n4 = ScalarParameter() {defaultValue=0.6, parameterName=ScaleGVertColor}
  n5 = Multiply(A=Time(), B=Multiply(A=ScalarParameter() {defaultValue=1, parameterName=Speed}, B=Multiply(A=@n3.G, B=@n4) {constB=0.5}))
  n6 = TextureSample(UVs=AppendVector(A=Add(A=@n2, B=@n5), B=ComponentMask(None=@n1) {g=true})) {samplerType=SAMPLERTYPE_Grayscale, texture=&LowResBlurredNoise}
  n7 = LinearInterpolate(B=ScalarParameter() {defaultValue=2.2, parameterName=WidthBase}, Alpha=@n6.G) {constA=1, constB=12}
  n8 = ComponentMask(None=TextureCoordinate()) {r=true}
  n9 = Multiply(A=OneMinus(None=@n8), B=@n8)
  n10 = LinearInterpolate(A=Multiply(A=@n7) {constB=4}, B=@n7, Alpha=@n9)
  n11 = VectorParameter() {parameterName=OffsetHeight, defaultValue=<rgba 0 0 150 0>}
  n12 = Multiply(A=@n11.RGB) {constB=-1}
  Add(
   A=MaterialFunctionCall(
    WidthBase=@n10
    WidthTip=@n10
    WorldPosition=ActorPositionWS()
    UVs for Thickness=TextureCoordinate()
    UVs for texturing=TextureCoordinate() {coordinateIndex=1}
   ) {materialFunction=&SplineThicken}.WorldPositionOffset
   B=Multiply(A=@n9, B=LinearInterpolate(A=@n12, B=@n11.RGB, Alpha=@n6.G))
  )
  n13 = Time()
  n14 = ScalarParameter() {defaultValue=2, parameterName=EmissivePulseSpeed}
  n15 = Panner(Coordinate=TextureCoordinate() {coordinateIndex=1, uTiling=0.5}, Time=Multiply(A=Multiply(A=@n13, B=@n14), B=VertexColor().G)) {speedY=1}
  n16 = LinearInterpolate(
   B=VectorParameter() {parameterName=EmissiveColorPulse1, defaultValue=<rgba 0 9.6517 16 0>}.RGB
   Alpha=Clamp(None=TextureSample(UVs=@n15) {texture=/Game/Global/Textures/Default/Masks_and_Gradients/T_Water_Top_Ramp}.G)
  )
  n17 = TextureSample(UVs=MaterialFunctionCall(XY=TextureCoordinate() {vTiling=0.05}) {materialFunction=&Swizzle}.YX) {texture=&T_Water_Top_Ramp}
  Add(A=@n16, B=Multiply(A=@n17.G, B=ScalarParameter() {defaultValue=0.6, parameterName=Non-Pulse_AlphaTweak}))
  n18 = LinearInterpolate(
   A=VectorParameter() {parameterName=EmissiveColorDark, defaultValue=<rgba 0 0.0417 0.0408 0>}.RGB
   B=VectorParameter() {parameterName=EmissiveColorBright, defaultValue=<rgba 0.0962 0.6087 1 0>}.RGB
   Alpha=@n17.G
  )
  n19 = Multiply(A=@n18) {constB=32}
  n20 = DepthFade(FadeDistance=ScalarParameter() {defaultValue=4, parameterName=ColorDepth})
  MaterialFunctionCall(Emissive=Add(A=@n16, B=LinearInterpolate(A=@n19, B=@n18, Alpha=@n20))) {materialFunction=&MF_FX_EmissiveAdjustments}
  TextureSample() {texture=/Game/Global/Textures/Default/Noise/TankMoteNoise}
 ]
}
